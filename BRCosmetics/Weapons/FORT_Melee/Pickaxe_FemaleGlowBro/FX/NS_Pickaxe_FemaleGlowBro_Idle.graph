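# Niagara idle effect for the FemaleGlowBro pickaxe cosmetic: a persistent, pulsing pink-red glow at socket-relative offsets.
# One FireGlow emitter bursts three non-expiring particles, partitioned to set each one's offset and visibility tag for sprite and mesh renderers.
# Glow alpha fades in with system age and pulses via a waveform.
# Alpha scales by quality level and front-end state; sprite size also shrinks in the front end.
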NiagaraSystem /BRCosmetics/Weapons/FORT_Melee/Pickaxe_FemaleGlowBro/FX/NS_Pickaxe_FemaleGlowBro_Idle
dependencies {
 usedDataInterfaces [&NiagaraDataInterfaceCurve, &NiagaraDataInterfaceArrayFloat, &NiagaraDataInterfaceSocketReader]
 usedDynamicInputs []
 usedModules [
  &self:SystemSpawnScript.SystemScriptSource.SystemScriptGraph.NiagaraNodeAssignment_2.SetVariables_#1
  &SystemState
 ]
}
emitters {FireGlow {
  emitterUpdateScript [
   &EmitterState(
    Inactive Response=::NewEnumerator1
    Life Cycle Mode=::NewEnumerator1
    Loop Behavior=::NewEnumerator1
    Loop Duration Mode=::NewEnumerator1
    UseLoopDelay=false
   )
   &SpawnBurst_Instantaneous(Spawn Count=3, Use Spawn Probability=false, UseLoopCountLimit=false)
  ]
  particleSpawnScript [
   &PartitionParticles(Number Of Partitions=3)
   /SetVariables(Particles.VisibilityTag=&MergeHelper_IntSwitch(0=0, 1=1, 2=1, 3=?, 4=?, 5=?, Index=$Particles.PartitionParticles.Partition, Num=::NewEnumerator1)) {isSetParameters=true}
   &InitializeParticle(
    Color Mode=::NewEnumerator0
    Lifetime=5
    Mesh Scale=<9 9.2 9>
    Mesh Scale Mode=::NewEnumerator4
    Sprite Size Mode=::NewEnumerator3
    Uniform Sprite Size=&Multiply_Float(A=27, B=&FN_FrontEndFloatModifier(Gameplay=1, Front End=0.75))
    UsePositionOffset=false
    Write Material Random=true
   )
   &DynamicMaterialParameters(
    Index 0 Param 1=1
    Index 0 Param 2=1
    Index 0 Param 3=1
    Index 0 Param 4=1
    Module.Index0Float0Write=true
    Module.Index0Float1Write=true
    Module.Index0Float2Write=true
    Module.Index0Float3Write=true
   )
   &SocketLocation(
    SamplePosOffset=true
    Sampled Position Offset=&MergeHelper_VectorSwitch(0=<-23.5 0 -8.8>, 1=<-7.8738 0 28>, 2=<-20 0 30>, 3=?, Index=$Particles.PartitionParticles.Partition, Num=::NewEnumerator1)
    Socket Reader=$System.Socket Reader
    Socket Sampling Mode=::NewEnumerator3
    VelocitySpeedLimit=false
   )
   /SetVariables(Particles.GlowDynamic:Vector4f=<2 1 1 1>) {isSetParameters=true}
   &InitialMeshOrientation(Mesh Orientation Mode=::NewEnumerator2, Orientation Vector=<0 -1 0>, Use Rotation Vector=false)
   /SetVariables(Particles.GlowCameraOffset=-15) {isSetParameters=true}
  ]
  particleUpdateScript [
   &ParticleState(Kill Particles When Lifetime Has Elapsed=false, Let Infinitely Lived Particles Die When Emitter Deactivates=true)
   /SetVariables(Particles.GlowColor=&MakeLinearColorFromVectorAndFloat(
     "Vector (RGB)"=&MakeVectorFromLinearColorRGB(Color=<rgba 1 0 0.2789 1>)
     "Float (Alpha)"=&Multiply_Float(
      A=&Multiply_Float(
       A=0.5
       B=&FloatFromCurve(
        FloatCurve=DataInterface() {propertyValues {Curve {defaultValue=1, keys [{tangentMode=RCTM_User, time=0, value=0}, {tangentMode=RCTM_User, time=0.15, value=1}]}}}
        CurveIndex=$Engine.System.Age
        Scale Curve=&Multiply_Float(
         A=&SelectFloatFromArray(Direct Array Index=$Engine.QualityLevel, Float Selection Array=DataInterface() {propertyValues {FloatData [0.75,0.75,1,1,1]}})
         B=&FN_FrontEndFloatModifier(Gameplay=1, Front End=0.5)
        )
       )
      )
      B=&Waveform("Waveform [1]"=::NewEnumerator0, "[1] Frequency"=2.5, "[1] Amplitude Min/Max"=<0.5 1>)
     )
    )) {isSetParameters=true}
  ]
  rendererClasses [&NiagaraSpriteRendererProperties, &NiagaraMeshRendererProperties]
  renderers [
   {rendererClass=&NiagaraSpriteRendererProperties}
   {rendererClass=&NiagaraMeshRendererProperties, rendererIndex=1}
  ]
 }}
summary {
 systemName=NS_Pickaxe_FemaleGlowBro_Idle
 emitters [{emitterName=FireGlow, rendererClasses [&NiagaraSpriteRendererProperties, &NiagaraMeshRendererProperties]}]
 userVariables [
  User.bIsFrontEnd {type=NiagaraBool}
  User.bIsFrontEndPreview {type=NiagaraBool}
 ]
}
